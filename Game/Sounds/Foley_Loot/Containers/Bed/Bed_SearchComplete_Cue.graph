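# Sound cue for completing a search of a bed loot container, classed as loot-collect foley.
# A quality-level root sends its first two slots to a shared modulated SearchEnd3 wave and its other two to SearchEnd_opt waves, each behind a default modulator.
# The shared modulator widens pitch upward and varies volume for variation.
# Prop-search attenuation and a concurrency override that stops the farthest instance limit range and overlap.
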SoundCue /Game/Sounds/Foley_Loot/Containers/Bed/Bed_SearchComplete_Cue
root=QualityLevel_0
nodes [
 {id=Modulator_0, children [WavePlayer_0]}
 {id=Modulator_1, children [WavePlayer_1]}
 {id=Modulator_15, children [WavePlayer_57], properties {pitchMax=1.2, pitchMin=1, volumeMax=1.07, volumeMin=0.92}}
 {id=QualityLevel_0, children [Modulator_15,Modulator_15,Modulator_0,Modulator_1]}
 {id=WavePlayer_0, properties {soundWaveAssetPtr=&/Bed_SearchEnd_opt}}
 {id=WavePlayer_1, properties {soundWaveAssetPtr=&Bed_SearchEnd_opt}}
 {id=WavePlayer_57, properties {soundWaveAssetPtr=&/Bed_SearchEnd3}}
]
properties {
 attenuationSettings=/Game/Sounds/Attenuation/Fort_Prop_Search_Att
 bOverrideConcurrency=true
 maxDistance=1300
 soundClassObject=&Fort_Loot_Collect
 volumeMultiplier=0.5
 concurrencyOverrides {maxCount=4, resolutionRule=StopFarthestThenPreventNew}
}
