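# Sound cue for searching a bed loot container, using prop-search attenuation and a concurrency override that prevents new instances.
# A quality-level root picks among four branches that shrink in variety.
# The first two mix randomized bed search waves with a louder layer of randomized gear foley, while the last two play a single "opt" wave.
# Modulators slightly vary pitch and volume on the bed waves, leaving the gear layer unmodulated.
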SoundCue /Game/Sounds/Foley_Loot/Containers/Bed/Bed_Search_Cue
root=QualityLevel_0
nodes [
 {id=Mixer_0, children [Modulator_0,Random_1], properties {inputVolume [0.3,1.2]}}
 {id=Mixer_13, children [Modulator_23,Random_12], properties {inputVolume [0.3,1.2]}}
 {id=Modulator_0, children [Random_0], properties {pitchMax=1.07, pitchMin=0.92, volumeMax=1.07, volumeMin=0.92}}
 {id=Modulator_1, children [WavePlayer_0], properties {pitchMax=1.07, pitchMin=0.92, volumeMax=1.07, volumeMin=0.92}}
 {id=Modulator_2, children [WavePlayer_7], properties {pitchMax=1.07, pitchMin=0.92, volumeMax=1.07, volumeMin=0.92}}
 {id=Modulator_23, children [Random_11], properties {pitchMax=1.07, pitchMin=0.92, volumeMax=1.07, volumeMin=0.92}}
 {id=QualityLevel_0, children [Mixer_13,Mixer_0,Modulator_1,Modulator_2]}
 {id=Random_0, children [WavePlayer_1,WavePlayer_2], properties {weights [1,1]}}
 {id=Random_1, children [WavePlayer_4,WavePlayer_5], properties {weights [1,1]}}
 {id=Random_11, children [WavePlayer_60, WavePlayer_64, WavePlayer_65], properties {weights [1,1,1]}}
 {id=Random_12, children [WavePlayer_58, WavePlayer_59, WavePlayer_67], properties {weights [1,1,1]}}
 {id=WavePlayer_0, properties {soundWaveAssetPtr=&/Bed_Search_opt}}
 {id=WavePlayer_1, properties {soundWaveAssetPtr=&/Bed_Search11}}
 {id=WavePlayer_2, properties {soundWaveAssetPtr=&/Bed_Search5}}
 {id=WavePlayer_4, properties {soundWaveAssetPtr=/Game/Sounds/Foley_Loot/Containers/Misc/Prop_Gear_Search_05}}
 {id=WavePlayer_5, properties {soundWaveAssetPtr=/Game/Sounds/Foley_Loot/Containers/Misc/Prop_Gear_Search_06}}
 {id=WavePlayer_58, properties {soundWaveAssetPtr=&Prop_Gear_Search_05}}
 {id=WavePlayer_59, properties {soundWaveAssetPtr=&Prop_Gear_Search_06}}
 {id=WavePlayer_60, properties {soundWaveAssetPtr=&Bed_Search11}}
 {id=WavePlayer_64, properties {soundWaveAssetPtr=&Bed_Search5}}
 {id=WavePlayer_65, properties {soundWaveAssetPtr=&/Bed_Search6}}
 {id=WavePlayer_67, properties {soundWaveAssetPtr=/Game/Sounds/Foley_Loot/Containers/Misc/Prop_Gear_Search_01}}
 {id=WavePlayer_7, properties {soundWaveAssetPtr=&Bed_Search_opt}}
]
properties {
 attenuationSettings=/Game/Sounds/Attenuation/Fort_Prop_Search_Att
 bOverrideConcurrency=true
 maxDistance=1300
 soundClassObject=&Fort_Loot_Collect
 volumeMultiplier=1
 concurrencyOverrides {resolutionRule=PreventNew}
}
